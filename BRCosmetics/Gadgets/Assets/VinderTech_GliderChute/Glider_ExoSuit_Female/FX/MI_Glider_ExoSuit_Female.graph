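# Base material, despite its MI_ name, in the FX folder of the female ExoSuit glider, flagged for use on Niagara mesh particles.
# It uses Default Lit shading with legacy pixel depth offset mode.
# Its graph holds only one vector parameter with a dark grey default and no visible connection to a material output, so it appears to be a minimal placeholder.
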Material /BRCosmetics/Gadgets/Assets/VinderTech_GliderChute/Glider_ExoSuit_Female/FX/MI_Glider_ExoSuit_Female
graph {
 properties {bUsedWithNiagaraMeshParticles=true, pixelDepthOffsetMode=PDOM_Legacy, usedShadingModels=Default Lit}
 nodes [VectorParameter() {defaultValue=<rgba 0.2159 0.2159 0.2159 0>}]
}
